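# Physics asset for the SlickSwish Fore medium male character body, built to simulate the swinging skirt.
# Left and right back and side capsule chains hang from small base bodies free-constrained to the pelvis, with limited joints down each chain and free links joining their lower segments.
# Pelvis, feet and right-bag capsules add body volume, while a huge root sphere, topping out near the origin, approximates a floor.
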PhysicsAsset /BRCosmetics/Characters/Player/Male/Medium/Bodies/M_MED_SlickSwish_Fore/Meshes/M_MED_SlickSwish_Fore_Physics_2_FE
bodies {
 BASE_dyn_skirt_back_l [{shapeType=Capsule, length=0.505, radius=0.505, shapeName=BASE_dyn_skirt_back_l_capsule, rotation=<rot 0 -90 0>}]
 BASE_dyn_skirt_back_r [{shapeType=Capsule, length=0.505, radius=0.505, shapeName=BASE_dyn_skirt_back_r_capsule, rotation=<rot 0 -90 0>}]
 BASE_dyn_skirt_m_l [{shapeType=Capsule, length=0.505, radius=0.505, shapeName=BASE_dyn_skirt_m_l_capsule, rotation=<rot 0 -90 0>}]
 BASE_dyn_skirt_m_r [{shapeType=Capsule, length=0.505, radius=0.505, shapeName=BASE_dyn_skirt_m_r_capsule, rotation=<rot 0 -90 0>}]
 Bag_l []
 Bag_r [{shapeType=Capsule, length=30.7021, radius=9.9558, center=<8.6202 -1.1058 1.5619>, rotation=<rot 0.0695 -88.3882 88.6879>}]
 calf_l []
 calf_r []
 dyn_skirt_back_1_l [{
   shapeType=Capsule
   length=19.5116
   radius=10.4414
   shapeName=dyn_skirt_back_1_l_capsule
   center=<9.1924 -7.2477 3.7418>
   rotation=<rot -25.0755 91.4727 1.2171>
  }]
 dyn_skirt_back_1_r [{
   shapeType=Capsule
   length=19.5116
   radius=10.4414
   shapeName=dyn_skirt_back_1_l_capsule
   center=<10.0169 -6.7944 -3.264>
   rotation=<rot 21.3236 92.4191 1.1834>
  }]
 dyn_skirt_back_2_l [{
   shapeType=Capsule
   length=19.5116
   radius=10.4414
   shapeName=dyn_skirt_back_1_l_capsule
   center=<9.3833 -4.4516 2.4423>
   rotation=<rot -25.0755 91.4727 1.2171>
  }]
 dyn_skirt_back_2_r [{
   shapeType=Capsule
   length=19.5116
   radius=10.4414
   shapeName=dyn_skirt_back_1_l_capsule
   center=<9.4764 -4.4704 -2.3585>
   rotation=<rot 21.3236 92.4191 1.1834>
  }]
 dyn_skirt_back_3_l [
  {
   shapeType=Capsule
   length=19.5116
   radius=10.4414
   shapeName=dyn_skirt_back_1_l_capsule
   center=<11.3607 -4.3829 2.4804>
   rotation=<rot -13.7439 91.719 1.1349>
  }
  {
   shapeType=Capsule
   length=19.5116
   radius=10.4414
   shapeName=dyn_skirt_back_3_l_capsule
   center=<26.7718 -4.1951 2.5846>
   rotation=<rot -13.7439 91.719 1.1349>
  }
 ]
 dyn_skirt_back_3_r [
  {
   shapeType=Capsule
   length=19.5116
   radius=10.4414
   shapeName=dyn_skirt_back_1_r_capsule
   center=<11.3939 -4.3433 -2.6383>
   rotation=<rot 13.7439 91.719 1.1349>
  }
  {
   shapeType=Capsule
   length=19.5116
   radius=10.4414
   shapeName=dyn_skirt_back_3_r_capsule
   center=<27.2431 -4.1953 -2.5257>
   rotation=<rot 13.7439 91.719 1.1349>
  }
 ]
 dyn_skirt_side_1_l [{
   shapeType=Capsule
   length=12.7234
   radius=7.4329
   shapeName=dyn_skirt_side_1_l_capsule
   center=<7.2955 -7.0422 2.4633>
   rotation=<rot 25.1133 -87.2231 1.8953>
  }]
 dyn_skirt_side_1_r [{
   shapeType=Capsule
   length=12.7234
   radius=7.4329
   shapeName=dyn_skirt_side_1_l_capsule
   center=<9.0721 -7.6539 -2.3834>
   rotation=<rot -18.9105 -90.197 0.7636>
  }]
 dyn_skirt_side_m_2_l [{
   shapeType=Capsule
   length=15.3669
   radius=8.3221
   shapeName=dyn_skirt_side_1_l_capsule
   center=<7.329 -7.7199 1.7236>
   rotation=<rot 21.0793 -89.9349 0.7449>
  }]
 dyn_skirt_side_m_2_r [{
   shapeType=Capsule
   length=16.3137
   radius=7.3753
   shapeName=dyn_skirt_side_1_l_capsule
   center=<7.278 -10.1054 0.5306>
   rotation=<rot -19.2312 -90.4454 0.7363>
  }]
 dyn_skirt_side_m_3_l [{
   shapeType=Capsule
   length=15.8742
   radius=7.8147
   shapeName=dyn_skirt_side_1_l_capsule
   center=<11.7105 -5.5644 2.9853>
   rotation=<rot 12.345 -90.0507 0.7115>
  }]
 dyn_skirt_side_m_3_r [{
   shapeType=Capsule
   length=14.5674
   radius=9.1216
   shapeName=dyn_skirt_side_1_r_capsule
   center=<10.329 -9.6978 1.3065>
   rotation=<rot 2.2867 -90.1751 -179.3044>
  }]
 dyn_skirt_side_m_4_l [{
   shapeType=Capsule
   length=19.0527
   radius=10.6085
   shapeName=dyn_skirt_side_1_l_capsule
   center=<7.2579 -9.7622 -3.5393>
   rotation=<rot 7.4362 -90.1121 0.701>
  }]
 dyn_skirt_side_m_4_r [{
   shapeType=Capsule
   length=19.0527
   radius=10.6085
   shapeName=dyn_skirt_side_1_r_capsule
   center=<11.4591 -10.7183 3.6112>
   rotation=<rot 7.4362 -90.1121 -179.299>
  }]
 dyn_skirt_side_m_5_l [
  {
   shapeType=Capsule
   length=13.6248
   radius=16.0364
   shapeName=dyn_skirt_side_1_l_capsule
   center=<14.6067 -11.9047 -4.1342>
   rotation=<rot -0.2882 -90.2063 0.6951>
  }
  {
   shapeType=Capsule
   length=11.9555
   radius=4.8322
   shapeName=dyn_skirt_side_m_5_l_capsule
   center=<28.9544 0.1939 6.3938>
   rotation=<rot 5.4079 -90.137 0.6982>
  }
 ]
 dyn_skirt_side_m_5_r [
  {
   shapeType=Capsule
   length=13.6248
   radius=16.0364
   shapeName=dyn_skirt_side_1_r_capsule
   center=<26.1062 -15.2197 3.9781>
   rotation=<rot -0.2882 -90.2063 -179.3049>
  }
  {
   shapeType=Capsule
   length=13.6248
   radius=16.0364
   shapeName=dyn_skirt_side_m_5_r_capsule
   center=<7.9176 -16.209 4.3145>
   rotation=<rot -0.2882 -90.2063 -179.3049>
  }
 ]
 foot_l [{shapeType=Capsule, length=15.2762, radius=8.5027, center=<-4.9855 6.0202 -0.2798>, rotation=<rot 0.1579 -178.8137 -89.1046>}]
 foot_r [{shapeType=Capsule, length=15.276, radius=8.5043, center=<3.777 -5.3163 0.3239>, rotation=<rot 0.191 -178.7952 -88.9167>}]
 pelvis [{shapeType=Capsule, length=5.4053, radius=12.5368, shapeName=pelvis_capsule, center=<0 0.8158 0>, rotation=<rot 0 -0.0002 0>}]
 root [{shapeType=Sphere, radius=668.1761, shapeName=root_sphere, center=<0 -28.9802 -666.2379>}]
}
constraints [
 {bone1Name=calf_l, bone2Name=pelvis, swing1LimitDegrees=45, swing2LimitDegrees=45, twistLimitDegrees=45}
 {bone1Name=foot_l, bone2Name=calf_l, swing1LimitDegrees=45, swing2LimitDegrees=45, twistLimitDegrees=45}
 {bone1Name=calf_r, bone2Name=pelvis, swing1LimitDegrees=45, swing2LimitDegrees=45, twistLimitDegrees=45}
 {bone1Name=foot_r, bone2Name=calf_r, swing1LimitDegrees=45, swing2LimitDegrees=45, twistLimitDegrees=45}
 {
  bone1Name=BASE_dyn_skirt_back_l
  bone2Name=pelvis
  swing1LimitDegrees=45
  swing1Motion=Free
  swing2LimitDegrees=45
  swing2Motion=Free
  twistLimitDegrees=45
  twistMotion=Free
 }
 {bone1Name=dyn_skirt_back_1_l, bone2Name=BASE_dyn_skirt_back_l, swing1LimitDegrees=45, swing2LimitDegrees=10, twistLimitDegrees=5}
 {bone1Name=dyn_skirt_back_2_l, bone2Name=dyn_skirt_back_1_l, swing1LimitDegrees=45, swing2LimitDegrees=10, twistLimitDegrees=5}
 {bone1Name=dyn_skirt_back_3_l, bone2Name=dyn_skirt_back_2_l, swing1LimitDegrees=45, swing2LimitDegrees=10, twistLimitDegrees=5}
 {
  bone1Name=BASE_dyn_skirt_back_r
  bone2Name=pelvis
  swing1LimitDegrees=45
  swing1Motion=Free
  swing2LimitDegrees=45
  swing2Motion=Free
  twistLimitDegrees=45
  twistMotion=Free
 }
 {bone1Name=dyn_skirt_back_1_r, bone2Name=BASE_dyn_skirt_back_r, swing1LimitDegrees=45, swing2LimitDegrees=10, twistLimitDegrees=5}
 {bone1Name=dyn_skirt_back_2_r, bone2Name=dyn_skirt_back_1_r, swing1LimitDegrees=45, swing2LimitDegrees=10, twistLimitDegrees=5}
 {bone1Name=dyn_skirt_back_3_r, bone2Name=dyn_skirt_back_2_r, swing1LimitDegrees=45, swing2LimitDegrees=10, twistLimitDegrees=5}
 {
  bone1Name=BASE_dyn_skirt_m_l
  bone2Name=pelvis
  swing1LimitDegrees=45
  swing1Motion=Free
  swing2LimitDegrees=45
  swing2Motion=Free
  twistLimitDegrees=45
  twistMotion=Free
 }
 {bone1Name=dyn_skirt_side_1_l, bone2Name=BASE_dyn_skirt_m_l, swing1LimitDegrees=45, swing2LimitDegrees=10, twistLimitDegrees=10}
 {bone1Name=dyn_skirt_side_m_2_l, bone2Name=dyn_skirt_side_1_l, swing1LimitDegrees=20, swing2LimitDegrees=10, twistLimitDegrees=5}
 {bone1Name=dyn_skirt_side_m_3_l, bone2Name=dyn_skirt_side_m_2_l, swing1LimitDegrees=20, swing2LimitDegrees=10, twistLimitDegrees=5}
 {bone1Name=dyn_skirt_side_m_4_l, bone2Name=dyn_skirt_side_m_3_l, swing1LimitDegrees=20, swing2LimitDegrees=10, twistLimitDegrees=5}
 {
  bone1Name=dyn_skirt_side_m_5_l
  bone2Name=dyn_skirt_side_m_4_l
  swing1LimitDegrees=20
  swing2LimitDegrees=10
  twistLimitDegrees=10
  twistMotion=Locked
 }
 {
  bone1Name=BASE_dyn_skirt_m_r
  bone2Name=pelvis
  swing1LimitDegrees=45
  swing1Motion=Free
  swing2LimitDegrees=45
  swing2Motion=Free
  twistLimitDegrees=45
  twistMotion=Free
 }
 {bone1Name=dyn_skirt_side_1_r, bone2Name=BASE_dyn_skirt_m_r, swing1LimitDegrees=45, swing2LimitDegrees=10, twistLimitDegrees=10}
 {bone1Name=dyn_skirt_side_m_2_r, bone2Name=dyn_skirt_side_1_r, swing1LimitDegrees=20, swing2LimitDegrees=10, twistLimitDegrees=5}
 {bone1Name=dyn_skirt_side_m_3_r, bone2Name=dyn_skirt_side_m_2_r, swing1LimitDegrees=20, swing2LimitDegrees=10, twistLimitDegrees=5}
 {bone1Name=dyn_skirt_side_m_4_r, bone2Name=dyn_skirt_side_m_3_r, swing1LimitDegrees=20, swing2LimitDegrees=10, twistLimitDegrees=5}
 {
  bone1Name=dyn_skirt_side_m_5_r
  bone2Name=dyn_skirt_side_m_4_r
  swing1LimitDegrees=20
  swing2LimitDegrees=10
  twistLimitDegrees=10
  twistMotion=Locked
 }
 {
  bone1Name=pelvis
  bone2Name=root
  swing1LimitDegrees=45
  swing1Motion=Free
  swing2LimitDegrees=45
  swing2Motion=Free
  twistLimitDegrees=45
  twistMotion=Free
 }
 {
  bone1Name=dyn_skirt_side_m_4_l
  bone2Name=dyn_skirt_back_3_l
  swing1LimitDegrees=45
  swing1Motion=Free
  swing2LimitDegrees=45
  swing2Motion=Free
  twistLimitDegrees=45
  twistMotion=Free
 }
 {
  bone1Name=dyn_skirt_side_m_4_r
  bone2Name=dyn_skirt_back_3_r
  swing1LimitDegrees=45
  swing1Motion=Free
  swing2LimitDegrees=45
  swing2Motion=Free
  twistLimitDegrees=45
  twistMotion=Free
 }
 {
  bone1Name=Bag_r
  bone2Name=pelvis
  swing1LimitDegrees=45
  swing1Motion=Free
  swing2LimitDegrees=45
  swing2Motion=Free
  twistLimitDegrees=45
  twistMotion=Free
 }
 {
  bone1Name=Bag_l
  bone2Name=pelvis
  swing1LimitDegrees=45
  swing1Motion=Free
  swing2LimitDegrees=45
  swing2Motion=Free
  twistLimitDegrees=45
  twistMotion=Free
 }
]
