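# Control Rig for the face accessory part of the M_MED_SlickSwish_Fore male cosmetic body, apparently an unfinished corrective for the right hood bone.
# The forward solve runs only a spherical pose reader on upperarm_r, with debug drawing enabled.
# A Modify Transforms node targets BASE_dyn_hood_r with an offset and Euler-derived rotation, but it is outside the execution chain, has zero weight and takes no input from the reader.
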ControlRigBlueprint /BRCosmetics/Characters/Player/Male/Medium/Bodies/M_MED_SlickSwish_Fore/Meshes/Parts/M_MED_SlickSwish_Fore_FaceAcc_CtrlRig
graphs [
 RigVMFunctionLibrary {links [], nodes []}
 RigVMModel {nodes [
   RigUnit_BeginExecution = &RigUnit_BeginExecution()
   &RigUnit_SphericalPoseReader(
    ActiveRegionScaleFactors=(PositiveWidth=0.3,NegativeWidth=0.3,PositiveHeight=0.3,NegativeHeight=0.3)
    ActiveRegionSize=0.3
    Debug=(bDrawDebug=True,bDraw2D=false,bDrawLocalAxes=false,DebugScale=30,DebugSegments=30,DebugThickness=0.2)
    DriverAxis=(X=-1,Y=0,Z=0)
    DriverItem=(Type=Bone,Name="upperarm_r")
    FalloffRegionScaleFactors=(PositiveWidth=0.1,NegativeWidth=1,PositiveHeight=1,NegativeHeight=0.5)
    FalloffSize=0.5
    OptionalParentItem=(Type=Bone,Name="None")
    RotationOffset=(X=-40,Y=220,Z=0)
    ExecutePin=@RigUnit_BeginExecution.ExecutePin
   )
   RigVMFunction_MathQuaternionFromEuler = &RigVMFunction_MathQuaternionFromEuler(Euler=(X=0,Y=100,Z=0))
   &RigUnit_ModifyTransforms(
    ExecutePin=()
    ItemToModify=((Item=(Type=Bone,Name="BASE_dyn_hood_r"),Transform=(Rotation=(X=0,Y=0,Z=0,W=1),Translation=(X=0,Y=0,Z=-6),Scale3D=(X=1,Y=1,Z=1))))
    Weight=0
    ItemToModify.0.Transform.Rotation=@RigVMFunction_MathQuaternionFromEuler.Result
   )
  ]}
]
hierarchy [
 root {loc=<0 0 0>, rot=<0 0 0 1>}
 pelvis {parent=root, loc=<0 -0.7071 93.1191>, rot=<0.0127 0.707 -0.0127 -0.707>}
 spine_01 {parent=pelvis, loc=<6.7494 -0.6388 0>, rot=<0 0 0.034 -0.9994>}
 spine_02 {parent=spine_01, loc=<7.2518 -0.035 0>, rot=<0 0 0.0238 -0.9997>}
 spine_03 {parent=spine_02, loc=<6.0104 0 0>, rot=<0 0 0.0149 -0.9999>}
 spine_04 {parent=spine_03, loc=<8.1024 0 0>, rot=<0 0 -0.0132 -0.9999>}
 spine_05 {parent=spine_04, loc=<10.0146 0 0>, rot=<0 0 -0.0015 -1>}
 neck_01 {parent=spine_05, loc=<11.3138 0.1529 0>, rot=<0 0 -0.2003 -0.9797>}
 neck_02 {parent=neck_01, loc=<4.5627 0 0>, rot=<0 0 0.0504 -0.9987>}
 head {parent=neck_02, loc=<5.3914 0 0>, rot=<0 0 0.115 -0.9934>}
 BASE_dyn_hood_l {parent=head, loc=<2.306 0.3563 -9.5468>, rot=<0 0 1 0>}
 dyn_hood_l {parent=BASE_dyn_hood_l, loc=<0 0 0>, rot=<0 0 0 1>}
 BASE_dyn_hood_r {parent=head, loc=<2.306 0.3563 9.0736>, rot=<0 0 1 0>}
 dyn_hood_r {parent=BASE_dyn_hood_r, loc=<0 0 0>, rot=<0 0 0 1>}
 clavicle_l {parent=spine_05, loc=<4.3575 -0.1466 -2.3223>, rot=<0.0465 -0.7289 -0.009 -0.6829>}
 upperarm_l {parent=clavicle_l, loc=<15.5111 0 0>, rot=<0.0226 -0.3234 -0.0088 -0.9459>}
 lowerarm_l {parent=upperarm_l, loc=<26.6514 0 0>, rot=<0.0391 -0.0099 -0.2453 -0.9686>}
 deform_elbow_in_base_l {parent=lowerarm_l, loc=<0.2112 5.259 0.8274>, rot=<0 0 0 1>}
 deform_elbow_in_l {parent=deform_elbow_in_base_l, loc=<0 0 0>, rot=<0 0 0 1>}
 upperarm_twist_01_l {parent=upperarm_l, loc=<0.0041 0 0>, rot=<0 0 0 -1>}
 deform_upperarm_bk_base_l {parent=upperarm_twist_01_l, loc=<1.0813 -7.2193 -1.9453>, rot=<0 0 0 -1>}
 deform_upperarm_bk_l {parent=deform_upperarm_bk_base_l, loc=<0 0 0>, rot=<0 0 0 1>}
 deform_upperarm_fr_base_l {parent=upperarm_twist_01_l, loc=<-0.3691 7.0506 -2.765>, rot=<0 0 0 -1>}
 deform_upperarm_fr_l {parent=deform_upperarm_fr_base_l, loc=<0 0 0>, rot=<0 0 0 1>}
 deform_upperarm_up_base_l {parent=upperarm_twist_01_l, loc=<-3.2216 -1.1391 3.3311>, rot=<0 0 0 1>}
 deform_upperarm_up_l {parent=deform_upperarm_up_base_l, loc=<0 0 0>, rot=<0 0 0 1>}
 upperarm_twist_02_l {parent=upperarm_l, loc=<13.7941 0 0>, rot=<0 0 0 -1>}
 clavicle_r {parent=spine_05, loc=<4.3575 -0.1466 2.3223>, rot=<0.7289 0.0465 -0.6829 0.009>}
 upperarm_r {parent=clavicle_r, loc=<-15.5111 0 0>, rot=<0.0226 -0.3234 -0.0088 -0.9459>}
 lowerarm_r {parent=upperarm_r, loc=<-26.6514 0 0>, rot=<0.0391 -0.0099 -0.2453 -0.9686>}
 deform_elbow_in_base_r {parent=lowerarm_r, loc=<-0.2111 -5.2591 -0.8275>, rot=<0 0 0 1>}
 deform_elbow_in_r {parent=deform_elbow_in_base_r, loc=<-0.0001 0 0.0002>, rot=<0 0 0 1>}
 upperarm_twist_01_r {parent=upperarm_r, loc=<-0.0041 0 0>, rot=<0 0 0 -1>}
 deform_upperarm_bk_base_r {parent=upperarm_twist_01_r, loc=<-1.0814 7.2192 1.9455>, rot=<0 0 0 -1>}
 deform_upperarm_bk_r {parent=deform_upperarm_bk_base_r, loc=<0 0 0>, rot=<0 0 0 1>}
 deform_upperarm_fr_base_r {parent=upperarm_twist_01_r, loc=<0.3689 -7.0506 2.7652>, rot=<0 0 0 -1>}
 deform_upperarm_fr_r {parent=deform_upperarm_fr_base_r, loc=<0 0 0>, rot=<0 0 0 1>}
 deform_upperarm_up_base_r {parent=upperarm_twist_01_r, loc=<3.2214 1.1391 -3.3308>, rot=<0 0 0 1>}
 deform_upperarm_up_r {parent=deform_upperarm_up_base_r, loc=<0 0 0>, rot=<0 0 0 -1>}
 upperarm_twist_02_r {parent=upperarm_r, loc=<-13.7941 0 0>, rot=<0 0 0 -1>}
 deform_collar_b_r {parent=spine_05, loc=<14.4604 -3.6476 6.8302>, rot=<0.3845 0.0701 -0.1457 -0.9088>}
 dyn_collar_b_r {parent=deform_collar_b_r, loc=<0 0 0>, rot=<0 0 0 -1>}
 deform_collar_r {parent=spine_05, loc=<11.6842 1.6728 8.8592>, rot=<0.7383 0.1003 -0.2175 -0.6306>}
 dyn_collar_r {parent=deform_collar_r, loc=<0 0 0>, rot=<0 0 0 -1>}
 deform_collar_f_r {parent=spine_05, loc=<8.7036 5.5727 7.0277>, rot=<0.7961 0.2019 -0.126 -0.5563>}
 dyn_collar_f_r {parent=deform_collar_f_r, loc=<0 0 0>, rot=<0 0 0 -1>}
 deform_collar_bk {parent=spine_05, loc=<15.1838 -5.4895 0>, rot=<0 0 -0.2003 -0.9797>}
 dyn_collar_bk {parent=deform_collar_bk, loc=<0 0 0>, rot=<0 0 0 1>}
 deform_collar_b_l {parent=spine_05, loc=<14.4604 -3.6476 -6.8302>, rot=<-0.3845 -0.0701 -0.1457 -0.9088>}
 dyn_collar_b_l {parent=deform_collar_b_l, loc=<0 0 0>, rot=<0 0 0 1>}
 deform_collar_l {parent=spine_05, loc=<11.6842 1.6728 -8.8592>, rot=<-0.7383 -0.1003 -0.2174 -0.6306>}
 dyn_collar_l {parent=deform_collar_l, loc=<0 0 0>, rot=<0 0 0 1>}
 deform_collar_f_l {parent=spine_05, loc=<8.7036 5.5727 -7.0277>, rot=<0.8825 0.2212 0.0881 0.4056>}
 dyn_collar_f_l {parent=deform_collar_f_l, loc=<0 0 0>, rot=<0 0 0 1>}
 deform_pec_aim_base_l {parent=spine_04, loc=<7.8732 12.6388 -2.4123>, rot=<0.0102 -0.4602 0.3979 -0.7936>}
 deform_pec_aim_l {parent=deform_pec_aim_base_l, loc=<0 0 0>, rot=<0 0 0 1>}
 deform_pec_base_l {parent=deform_pec_aim_l, loc=<10.766 0 0>, rot=<0 0 0 1>}
 deform_pec_l {parent=deform_pec_base_l, loc=<0 0 0>, rot=<0 0 0 1>}
 deform_pec_aim_base_r {parent=spine_04, loc=<7.8727 12.6388 2.4123>, rot=<-0.4602 -0.0102 0.7936 0.3979>}
 deform_pec_aim_r {parent=deform_pec_aim_base_r, loc=<0 0 0>, rot=<0 0 0 1>}
 deform_pec_base_r {parent=deform_pec_aim_r, loc=<-10.7662 -0.0006 -0.0007>, rot=<0 0 0 1>}
 deform_pec_r {parent=deform_pec_base_r, loc=<0 -0.0003 -0.0004>, rot=<0 0 0 1>}
 thigh_l {parent=pelvis, loc=<-2.5693 -0.0109 -7.7501>, rot=<0.0012 0.0572 0.0155 -0.9982>}
 calf_l {parent=thigh_l, loc=<-39.0484 0 0>, rot=<0.0006 -0.0161 -0.0391 -0.9991>}
 foot_l {parent=calf_l, loc=<-42.561 0.1077 0.4292>, rot=<0.0418 -0.0455 0.0343 -0.9975>}
 calf_twist_02_l {parent=calf_l, loc=<-18.8064 -0.0967 0.0397>, rot=<0.0237 0.0095 -0.0035 -0.9997>}
 calf_twist_01_l {parent=calf_l, loc=<-35.813 -0.0967 0.0397>, rot=<0.0237 0.0095 -0.0035 -0.9997>}
 deform_knee_in_base_l {parent=calf_l, loc=<1.298 -5.2957 -0.0496>, rot=<0 0 0 -1>}
 deform_knee_in_l {parent=deform_knee_in_base_l, loc=<0 0 0>, rot=<0 0 0 1>}
 deform_knee_out_base_l {parent=calf_l, loc=<0 4 0>, rot=<0 0 0 -1>}
 deform_knee_out_l {parent=deform_knee_out_base_l, loc=<0 0 0>, rot=<0 0 0 1>}
 thigh_twist_01_l {parent=thigh_l, loc=<-21.6303 0 0>, rot=<0.0013 0 0 -1>}
 deform_groin_base_l {parent=thigh_l, loc=<-2.1908 8.1142 -1.7442>, rot=<0 0 0 -1>}
 deform_groin_l {parent=deform_groin_base_l, loc=<0 0 0>, rot=<0 0 0 1>}
 thigh_r {parent=pelvis, loc=<-2.5693 -0.0109 7.7501>, rot=<0.0572 -0.0012 0.9982 0.0155>}
 calf_r {parent=thigh_r, loc=<39.0484 0 0>, rot=<0.0006 -0.0161 -0.0391 -0.9991>}
 foot_r {parent=calf_r, loc=<42.561 -0.1077 -0.4292>, rot=<0.0418 -0.0455 0.0343 -0.9975>}
 calf_twist_02_r {parent=calf_r, loc=<18.8064 0.0967 -0.0397>, rot=<0.0237 0.0095 -0.0035 -0.9997>}
 calf_twist_01_r {parent=calf_r, loc=<35.813 0.0967 -0.0397>, rot=<0.0237 0.0095 -0.0035 -0.9997>}
 deform_knee_in_base_r {parent=calf_r, loc=<-1.298 5.2957 0.0496>, rot=<0 0 0 1>}
 deform_knee_in_r {parent=deform_knee_in_base_r, loc=<0 0 0>, rot=<0 0 0 1>}
 deform_knee_out_base_r {parent=calf_r, loc=<0 -4 0>, rot=<0 0 0 -1>}
 deform_knee_out_r {parent=deform_knee_out_base_r, loc=<0 0 0>, rot=<0 0 0 1>}
 calf_r_extra {parent=calf_r, loc=<0 0 0>, rot=<0 0 0 1>}
 thigh_twist_01_r {parent=thigh_r, loc=<21.6303 0 0>, rot=<0.0013 0 0 -1>}
 deform_groin_base_r {parent=thigh_r, loc=<2.1908 -8.1142 1.7442>, rot=<0 0 0 -1>}
 deform_groin_r {parent=deform_groin_base_r, loc=<0 0 0>, rot=<0 0 0 1>}
 deform_glute_base_l {parent=pelvis, loc=<-8.8064 -8.8222 -8.4287>, rot=<0.0275 0.0501 -0.4472 -0.8926>}
 deform_glute_l {parent=deform_glute_base_l, loc=<0 0 0>, rot=<0 0 0 1>}
 deform_glute_base_r {parent=pelvis, loc=<-8.8062 -8.8222 8.4287>, rot=<0.0501 -0.0275 0.8926 -0.4472>}
 deform_glute_r {parent=deform_glute_base_r, loc=<0 0 0>, rot=<0 0 0 1>}
 BASE_dyn_skirt_m_l {parent=pelvis, loc=<9.7724 0.201 -13.0903>, rot=<0.1173 -0.8283 -0.5457 -0.0478>}
 dyn_skirt_side_1_l {parent=BASE_dyn_skirt_m_l, loc=<0 0 0>, rot=<0 0 0 1>}
 dyn_skirt_side_m_2_l {parent=dyn_skirt_side_1_l, loc=<10.8732 0 0>, rot=<0 0 0 1>}
 dyn_skirt_side_m_3_l {parent=dyn_skirt_side_m_2_l, loc=<10.8732 0 0>, rot=<0 0 0 1>}
 dyn_skirt_side_m_4_l {parent=dyn_skirt_side_m_3_l, loc=<16.3098 0 0>, rot=<0 0 0 1>}
 dyn_skirt_side_m_5_l {parent=dyn_skirt_side_m_4_l, loc=<19.5636 0 0>, rot=<0 0 0 1>}
 dyn_skirt_side_m_l_end {parent=dyn_skirt_side_m_5_l, loc=<18.6891 0 0>, rot=<0 0 0 1>}
 BASE_dyn_skirt_back_l {parent=pelvis, loc=<4.3669 -11.3409 -8.9323>, rot=<0.0857 -0.9961 0.0023 -0.0184>}
 dyn_skirt_back_1_l {parent=BASE_dyn_skirt_back_l, loc=<0 0 0>, rot=<0 0 0 1>}
 dyn_skirt_back_2_l {parent=dyn_skirt_back_1_l, loc=<14.8306 0 0>, rot=<0 0 0 1>}
 dyn_skirt_back_3_l {parent=dyn_skirt_back_2_l, loc=<16.8072 0 0>, rot=<0 0 0 1>}
 dyn_skirt_back_l_end {parent=dyn_skirt_back_3_l, loc=<17.8411 0 0>, rot=<0 0 0 1>}
 BASE_dyn_skirt_back_r {parent=pelvis, loc=<4.3669 -11.3409 8.9323>, rot=<0.0857 -0.9961 0.0023 -0.0184>}
 dyn_skirt_back_1_r {parent=BASE_dyn_skirt_back_r, loc=<0 0 0>, rot=<0 -0.0371 0 -0.9993>}
 dyn_skirt_back_2_r {parent=dyn_skirt_back_1_r, loc=<14.1142 0 0>, rot=<0 0 0 1>}
 dyn_skirt_back_3_r {parent=dyn_skirt_back_2_r, loc=<16.7611 0 0>, rot=<0 0 0 1>}
 dyn_skirt_back_r_end {parent=dyn_skirt_back_3_r, loc=<17.792 0 0>, rot=<0 0 0 1>}
 BASE_dyn_skirt_m_r {parent=pelvis, loc=<9.7724 0.201 13.0903>, rot=<0.1119 -0.799 0.5894 0.0409>}
 dyn_skirt_side_1_r {parent=BASE_dyn_skirt_m_r, loc=<0 0 0>, rot=<0 0 0 1>}
 dyn_skirt_side_m_2_r {parent=dyn_skirt_side_1_r, loc=<10.8732 0 0>, rot=<0 0 0 1>}
 dyn_skirt_side_m_3_r {parent=dyn_skirt_side_m_2_r, loc=<10.8732 0 0>, rot=<0 0 0 1>}
 dyn_skirt_side_m_4_r {parent=dyn_skirt_side_m_3_r, loc=<16.3098 0 0>, rot=<0 0 0 1>}
 dyn_skirt_side_m_5_r {parent=dyn_skirt_side_m_4_r, loc=<19.5636 0 0>, rot=<0 0 0 1>}
 dyn_skirt_side_m_r_end {parent=dyn_skirt_side_m_5_r, loc=<18.6891 0 0>, rot=<0 0 0 1>}
]
