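# Physics asset for the face accessory part of the M_MED_SlickSwish_Fore medium male player cosmetic body, simulating its hood bones.
# It has a capsule body on the head and a matching capsule on each of the left and right dyn_hood bones.
# Each hood bone is constrained directly to the head with identical, narrow swing and twist limits, giving the hood restrained secondary motion around the head.
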PhysicsAsset /BRCosmetics/Characters/Player/Male/Medium/Bodies/M_MED_SlickSwish_Fore/Meshes/Parts/M_MED_SlickSwish_Fore_FaceAcc_Physics
bodies {
 dyn_hood_l [{shapeType=Capsule, length=5.84, radius=1.6284, shapeName=dyn_hood_l_capsule, center=<4.5078 0 0>, rotation=<rot 0 90 -90>}]
 dyn_hood_r [{shapeType=Capsule, length=5.84, radius=1.6284, shapeName=dyn_hood_l_capsule, center=<4.5078 0 0>, rotation=<rot 0 90 -90>}]
 head [{shapeType=Capsule, length=5.0813, radius=4.4694}]
}
constraints [
 {bone1Name=dyn_hood_l, bone2Name=head, swing1LimitDegrees=10, swing2LimitDegrees=30, twistLimitDegrees=5}
 {bone1Name=dyn_hood_r, bone2Name=head, swing1LimitDegrees=10, swing2LimitDegrees=30, twistLimitDegrees=5}
]
